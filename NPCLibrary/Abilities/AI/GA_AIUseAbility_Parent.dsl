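; Parent gameplay ability for Fortnite AI bots that prepares the AI when it uses an ability.
; On activation it caches the avatar pawn, AI bot controller and speech bubble component, ending the ability if either cast fails.
; Flag-driven setup then optionally reloads the weapon, pushes a disabling behavior control and stops fire.
; On end it pops that behavior control if it was pushed and the controller is valid.
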(event Ability|EventActivateAbility
  (CallFunction|InitVariables)
  (CallFunction|SetupAIForAbility))

(event Ability|EventOnEndAbility
  (params (bWasCancelled :type Boolean))
  (bind _aicontroller (Variables|InternalData|GetAIController))
  (Utilities|IsValid _aicontroller
    (:Is%20Valid
      (if (and (Variables|AI|GetShouldPushDisabledBehavior) (Variables|InternalData|GetBehaviorPushed))
        (AI|PopBehaviorControl _aicontroller)))))

(fn InitVariables
  (Variables|InternalData|SetBehaviorPushed false)
  (bind _outfortplayerpawn (CallFunction|GetAvatarasValidFortPlayerPawn)
    (:ValidFortPlayerPawn
      (bind _output_get (Variables|InternalData|SetAIPawn _outfortplayerpawn))
      (bind _asfort_athena_aibot_controller (Utilities|Casting|CastToFortAthenaAIBotController (Pawn|GetController _output_get))
        (:then
          (Variables|InternalData|SetAIController _asfort_athena_aibot_controller)
          (Variables|InternalData|SetSpeechBubbleComponent (Actor|GetComponentbyClass (Variables|InternalData|GetAIPawn) "/Script/FortniteGame.FortPawnComponent_SpeechBubble"))
          (return))
        (:CastFailed
          _)))
    (:AvatarCastFailed
      _))
  (Ability|EndAbility))

(fn SetupAIForAbility
  (bind _aicontroller (Variables|InternalData|GetAIController))
  (if (Variables|AI|GetShouldForceReloadWeapon)
    (Class|BPFLAICommonFunctions|ReloadPawnWeapon (Variables|InternalData|GetAIPawn))
    (else
      _))
  (if (Variables|AI|GetShouldPushDisabledBehavior)
    (AI|PushBehaviorControl _aicontroller -2097152)
    (Variables|InternalData|SetBehaviorPushed true)
    (else
      _))
  (if (Variables|AI|GetShouldStopAIFromShooting)
    (Class|FortAthenaAIBotController|StopFire _aicontroller)
    (else
      _))
  (return))
